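# Material function that layers a stylized ice look over incoming material attributes through a selectable, invertible mask channel, also output as IceMask.
# Bump offset and reflection-vector UV math fake inner depth, while crack and Perlin noise textures build frost and fresnel highlights.
# Deep and surface ice colors blend by an AO channel.
# Refracted skylight sampling, fake subsurface scattering, GGX sun specular and time-of-day scaling add environment-driven color and emissive lighting.
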MaterialFunction /Game/Global/Materials/MaterialGym/Materials/SimpleIce/MF_SimpleIce
graph {nodes [
  Constant() {r=1}
  n1 = ScalarParameter() {defaultValue=1, group=SimpleIce, parameterName=OriginalBaseColorBlend}
  n2 = Multiply(
   A=LinearInterpolate(A=NamedRerouteUsage(), B=NamedRerouteUsage(), Alpha=Saturate(None=@n1)) {constAlpha=0.9}
   B=ScalarParameter() {defaultValue=1, group=SimpleIce, parameterName=BaseColorIntensity}
  )
  n3 = ScalarParameter() {group=SimpleIce, parameterName=Metalness}
  n4 = Multiply(A=NamedRerouteUsage(), B=ScalarParameter() {defaultValue=1, group=SimpleIce, parameterName=SpecularMultiplier})
  n5 = ScalarParameter() {defaultValue=0.1, group=SimpleIce, parameterName=Roughness}
  n6 = NamedRerouteUsage()
  n7 = Constant() {r=1}
  n8 = NamedRerouteUsage()
  n9 = Constant() {r=1}
  n10 = ChannelMaskParameter(None=NamedRerouteUsage()) {group=SimpleIce, maskChannel=Blue, parameterName=IceFXMask, defaultValue=<rgba 0 0 1 0>}
  n11 = LinearInterpolate(A=@n10.R, B=OneMinus(None=@n10.R), Alpha=ScalarParameter() {group=SimpleIce, parameterName=IceFXMaskInvert})
  FunctionOutput(None=BlendMaterialAttributes(
    A=NamedRerouteUsage()
    B=MakeMaterialAttributes!(
     BaseColor=@n2
     Metallic=@n3
     Specular=@n4
     Roughness=@n5
     Anisotropy
     EmissiveColor=@n6
     Opacity=@n7
     OpacityMask
     Normal=@n8
     Tangent
     WorldPositionOffset
     SubsurfaceColor
     ClearCoat
     ClearCoatRoughness
     AmbientOcclusion=@n9
     Refraction
     CustomizedUVs_0
     CustomizedUVs_1
     CustomizedUVs_2
     CustomizedUVs_3
     CustomizedUVs_4
     CustomizedUVs_5
     CustomizedUVs_6
     CustomizedUVs_7
     PixelDepthOffset
     ShadingModel
     Displacement
    )
    Alpha=@n11
   ))
  FunctionOutput(None=@n11) {outputName=IceMask, sortPriority=1}
  n12 = Power(
   Base=ChannelMaskParameter(None=NamedRerouteUsage()) {group=SimpleIce, maskChannel=Green, parameterName=GMapChannel, defaultValue=<rgba 0 1 0 0>}.R
   Exp=ScalarParameter() {defaultValue=2, group=SimpleIce, parameterName=GMapExpnent}
  ) {constExponent=4}
  n13 = ScalarParameter() {defaultValue=1, group=SimpleIce, parameterName=GMapIntensity}
  n14 = LinearInterpolate(B=ScalarParameter() {defaultValue=1, group=SimpleIce, parameterName=RoughnessHighlightIntensity}, Alpha=NamedRerouteUsage()) {constB=1.8}
  n15 = Multiply(A=LinearInterpolate(Alpha=NamedRerouteUsage()) {constA=0.8}, B=LinearInterpolate(Alpha=@n14) {constA=0.05})
  n16 = VectorParameter() {group=SimpleIce, parameterName=Emissive color, defaultValue=<rgba 0.1686 0.2782 0.4375 1>}
  n17 = Multiply(A=@n16.RGB) {constB=1.5}
  n18 = Power(Base=@n14, Exp=ScalarParameter() {defaultValue=8, group=SimpleIce, parameterName=RoughnessHighlightExp}) {constExponent=8}
  n19 = Add(A=NamedRerouteUsage(), B=Multiply(A=NamedRerouteUsage()) {constB=0.5})
  n20 = Saturate(None=Multiply(A=Fresnel() {baseReflectFraction=0}) {constB=3})
  n21 = LinearInterpolate(A=Multiply(A=Add(A=@n19, B=@n20), B=@n16.RGB), B=Add(A=Multiply(A=@n15, B=@n16.RGB), B=Multiply(A=@n17, B=@n18)), Alpha=Constant() {r=0.5}) {constAlpha=0.7}
  Multiply(A=Saturate(None=Multiply(A=@n12, B=@n13)), B=@n21) {desc=In case we need to incorporate AO or GMap}
  NamedRerouteDeclaration(None=FunctionInput() {inputType=FunctionInput_MaterialAttributes}) {name=InMaterialAttributes, nodeColor=<rgba 1 0 0.3047 1>}
  n22 = Add(A=NamedRerouteUsage(), B=Constant2Vector() {r=0.4})
  n23 = TextureSample(UVs=Multiply(A=@n22, B=ScalarParameter() {defaultValue=4, group=SimpleIce, parameterName=ForsetScale}) {constB=4}) {
   samplerSource=SSM_Wrap_WorldGroupSettings
   samplerType=SAMPLERTYPE_Masks
   texture=/Game/Environments/Asteria/Terrain/Textures/T_Asteria_Ice_Cracks_M
  }
  n24 = TextureSample(UVs=@n22) {
   samplerSource=SSM_Wrap_WorldGroupSettings
   samplerType=SAMPLERTYPE_LinearColor
   texture=&/T_Perlin_Noise_M1
  }
  n25 = Add(A=Multiply(A=Power(Base=@n23.R) {constExponent=1.5}) {constB=0.25}, B=Multiply(A=@n24.R) {constB=0.75})
  n26 = NamedRerouteUsage()
  n27 = Power(
   Base=Fresnel(BaseReflectFractionIn=Constant(), Normal=Transform(None=@n26)) {exponent=1}
   Exp=ScalarParameter() {defaultValue=1, group=SimpleIce, parameterName=HighlightFresnelExponent}
  ) {constExponent=1.2}
  n28 = ScalarParameter() {defaultValue=10, group=SimpleIce, parameterName=HighLightEmissive}
  NamedRerouteDeclaration(None=Multiply(A=@n25, B=Multiply(A=@n27, B=@n28))) {name=Edge highlight, nodeColor=<rgba 1 0 0.4922 1>}
  n29 = GetMaterialAttributes(None=NamedRerouteUsage()) {attributeGetTypes [
    #1
    #2
    #3
    #4
    #5
   ]}
  NamedRerouteDeclaration(None=@n29.Specular) {name=InSpecular, nodeColor=<rgba 0.0312 0 1 1>}
  n30 = VertexNormalWS()
  n31 = CameraVectorWS()
  n32 = DotProduct(A=Normalize(VectorInput=Add(A=@n30, B=@n31)), B=Multiply(A=AtmosphericLightVector()) {constB=-1})
  n33 = VertexNormalWS()
  n34 = CameraVectorWS()
  n35 = Multiply(A=Add(A=Multiply(A=@n32) {constB=0.5}) {constB=0.5}, B=Saturate(None=DotProduct(A=@n33, B=@n34)))
  n36 = ScalarParameter() {defaultValue=4, group=SimpleIce, parameterName=FakeSSSExponent}
  NamedRerouteDeclaration(None=Saturate(None=Power(Base=@n35, Exp=@n36))) {name=FakeSSS, nodeColor=<rgba 0 0.0859 1 1>}
  n37 = MaterialFunctionCall!(
   Ray Direction=Multiply(A=CameraVectorWS()) {constB=-1}
   Surface Normal=PixelNormalWS()
   Refractive Index Origin
   Refractive Index Target=ScalarParameter() {defaultValue=3, group=SimpleIce, parameterName=RefractionAmount}
  ) {materialFunction=/Engine/Functions/Engine_MaterialFunctions01/Vectors/Refract}
  n38 = Multiply(A=Subtract(A=NamedRerouteUsage()) {constB=0.5}) {constB=0}
  NamedRerouteDeclaration(None=SkyLightEnvMapSample(Direction=Add(A=@n37.Vector, B=@n38))) {name=EnvironmentMap, nodeColor=<rgba 0 1 0.0625 1>}
  n39 = Saturate(None=Multiply(A=@n25) {constB=1.5})
  NamedRerouteDeclaration(None=@n39) {name=Forset, nodeColor=<rgba 1 0 0.1641 1>}
  NamedRerouteDeclaration(None=FunctionInput() {bUsePreviewValueAsDefault=true, inputName=InMask, sortPriority=1, previewValue=<1 1 1 1>}) {name=InMask, nodeColor=<rgba 1 0.7969 0 1>}
  n40 = VectorParameter() {group=SimpleIce, parameterName=IceDeepColor, defaultValue=<rgba 0.0237 0.0762 0.1198 1>}
  n41 = VertexInterpolator(VS=Divide(A=LocalPosition() {localOrigin=InstancePreSkinning}.XYZ) {constB=75})
  n42 = TextureSample(UVs=ComponentMask(None=NamedRerouteDeclaration(None=@n41.PS) {name=PreSkinnedPosition, nodeColor=<rgba 0 0.7891 1 1>}) {b=true, r=true}) {samplerSource=SSM_Wrap_WorldGroupSettings, samplerType=SAMPLERTYPE_Grayscale, texture=&LowResBlurredNoise}
  n43 = Multiply(
   A=LinearInterpolate(Alpha=Power(Base=@n42.R) {constExponent=1.2}) {constA=0.3}
   B=VectorParameter() {group=SimpleIce, parameterName=IceSurfaceColor, defaultValue=<rgba 0.1211 0.2339 0.375 1>}.RGB
  )
  n44 = ChannelMaskParameter(None=NamedRerouteUsage()) {group=SimpleIce, maskChannel=Green, parameterName=AOChannel, defaultValue=<rgba 0 1 0 0>}
  n45 = ScalarParameter() {defaultValue=2, group=SimpleIce, parameterName=AOExponent}
  n46 = Clamp(None=Multiply(A=ScalarParameter() {defaultValue=1, group=SimpleIce, parameterName=AOIntensity}, B=Power(Base=@n44.R, Exp=@n45) {constExponent=5}))
  NamedRerouteDeclaration(None=Multiply(A=LinearInterpolate(A=@n40.RGB, B=@n43, Alpha=@n46), B=ScalarParameter() {defaultValue=1, group=SimpleIce, parameterName=IceColorIntensity})) {name=ICE Color, nodeColor=<rgba 0.6875 1 0 1>}
  n47 = Multiply(
   A=LinearInterpolate(A=TextureCoordinate(), B=TextureCoordinate() {coordinateIndex=1}, Alpha=ScalarParameter() {group=SimpleIce, parameterName=UseUV2})
   B=ScalarParameter() {defaultValue=0.5, group=SimpleIce, parameterName=UVScale}
  ) {constB=2.5}
  n48 = TextureSample(UVs=Multiply(A=@n47)) {
   samplerSource=SSM_Wrap_WorldGroupSettings
   samplerType=SAMPLERTYPE_LinearColor
   texture=&T_Perlin_Noise_M1
  }
  n49 = MaterialFunctionCall!(
   Normal=MaterialFunctionCall(Normal=NamedRerouteUsage(), Flatness=Constant() {r=0.5}) {materialFunction=&FlattenNormal}.Result
   CameraVector=Transform(None=CameraVectorWS()) {transformSourceType=TRANSFORMSOURCE_World, transformType=TRANSFORM_Tangent}
  ) {materialFunction=/Engine/Functions/Engine_MaterialFunctions02/Utility/CustomReflectionVector}
  n50 = ComponentMask(None=@n49.Reflection Vector) {b=true}
  n51 = Multiply(
   A=ComponentMask(None=@n49.Reflection Vector) {g=true, r=true}
   B=Divide(A=ScalarParameter() {defaultValue=40, group=SimpleIce, parameterName=InnerIceDepth}, B=Abs(None=@n50)) {constA=100}
  )
  n52 = Divide(B=Constant() {r=512}) {constB=512}
  NamedRerouteDeclaration(None=Add(A=BumpOffset(Coordinate=@n47, Height=@n48.G) {heightRatio=0.1}, B=Multiply(A=@n51, B=@n52))) {name=Ice uv, nodeColor=<rgba 0.7578 1 0 1>}
  n53 = NamedRerouteUsage()
  n54 = ScalarParameter() {defaultValue=1, group=SimpleIce, parameterName=BaseColorEnvironmentAmount}
  n55 = NamedRerouteUsage()
  n56 = ScalarParameter() {defaultValue=2, group=SimpleIce, parameterName=BaseColorForsetExp}
  n57 = Saturate(None=Multiply(A=Power(Base=@n55, Exp=@n56) {constExponent=1}, B=ScalarParameter() {defaultValue=2, group=SimpleIce, parameterName=BaseColorForsetInt}))
  n58 = Add(A=LinearInterpolate(A=Multiply(A=@n53, B=@n54) {constB=0.5}, Alpha=@n57) {constAlpha=0.1, constB=0}, B=@n57)
  n59 = NamedRerouteUsage()
  n60 = TextureCoordinate()
  n61 = TextureCoordinate() {coordinateIndex=1}
  n62 = ScalarParameter() {group=SimpleIce, parameterName=UseUV2}
  n63 = TextureSample(UVs=Multiply(A=LinearInterpolate(A=@n60, B=@n61, Alpha=@n62), B=ScalarParameter() {defaultValue=2, group=SimpleIce, parameterName=Frost Tiling})) {
   samplerSource=SSM_Wrap_WorldGroupSettings
   samplerType=SAMPLERTYPE_LinearColor
   texture=&T_Perlin_Noise_M1
  }
  n64 = NamedRerouteDeclaration(None=Power(Base=@n63.RGB) {constExponent=5}) {name=Frost mask, nodeColor=<rgba 0.4531 1 0 1>}
  NamedRerouteDeclaration(None=LinearInterpolate(A=Multiply(A=@n58, B=@n59), Alpha=Multiply(B=@n64) {constA=0.15})) {name=Base Color, nodeColor=<rgba 0.0781 1 0 1>}
  n65 = NamedRerouteUsage()
  NamedRerouteDeclaration(None=LinearInterpolate(A=ScalarParameter() {group=SimpleIce, parameterName=IceDepthColorIntensity}, Alpha=Max(A=@n39, B=@n65))) {name=IceHighlight, nodeColor=<rgba 0.2656 1 0 1>}
  n66 = SkyAtmosphereLightDirection()
  n67 = CameraVectorWS()
  n68 = MaterialFunctionCall!(Normal=PixelNormalWS(), Light Vector=Normalize(VectorInput=Add(A=@n66, B=@n67)), Roughness=Constant() {r=0.05}) {materialFunction=/Engine/Functions/Engine_MaterialFunctions01/Lighting/GGXSpecular}
  n69 = Multiply(A=@n68.Result, B=@n68.Result)
  n70 = LinearInterpolate(
   A=@n21
   B=Multiply(A=@n21, B=VectorParameter() {group=SimpleIce, parameterName=FakeSSSColor, defaultValue=<rgba 2.5625 2.8656 3 1>}.RGB) {constB=3}
   Alpha=NamedRerouteUsage()
  )
  n71 = Constant() {r=0.8}
  n72 = MaterialFunctionCall(Color_Morning=@n71, Color_Day=Constant() {r=1}, Color_Evening=@n71, Color_Night=Constant() {r=0.6}) {materialFunction=&Color_Time_of_Day}
  NamedRerouteDeclaration(None=Add(A=Saturate(None=@n69), B=Multiply(A=@n70, B=@n72.Result))) {name=Emissive Color, nodeColor=<rgba 1 0.3516 0 1>}
  NamedRerouteDeclaration(None=FunctionInput() {bUsePreviewValueAsDefault=true, inputName=InFXMask, sortPriority=2, previewValue=<1 1 1 1>}) {name=InFXMask, nodeColor=<rgba 1 0.7969 0 1>}
  NamedRerouteDeclaration(None=@n29.BaseColor) {name=InBaseColor, nodeColor=<rgba 0.875 1 0 1>}
  NamedRerouteDeclaration(None=@n29.Normal) {name=InNormalMap, nodeColor=<rgba 0.8047 0 1 1>}
  NamedRerouteDeclaration(None=@n29.Roughness) {name=InRoughness, nodeColor=<rgba 1 0.3516 0 1>}
 ]}
