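# M_SimpleIce is a simple ice surface material in the MaterialGym folder, using Default Lit shading and material attributes.
# All shading logic lives in the MF_SimpleIce material function, which receives a GetMaterialAttributes input plus mask and FX mask inputs.
# It is flagged for skeletal mesh and mesh deformer use, with automatic usage setting in the editor turned off.
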Material /Game/Global/Materials/MaterialGym/Materials/SimpleIce/M_SimpleIce
graph {
 properties {
  bAutomaticallySetUsageInEditor=false
  bUseMaterialAttributes=true
  bUsedWithMeshDeformer=true
  bUsedWithSkeletalMesh=true
  usedShadingModels=Default Lit
 }
 nodes [MaterialFunctionCall!(In=GetMaterialAttributes().MaterialAttributes, InMask, InFXMask) {materialFunction=&/MF_SimpleIce}]
}
